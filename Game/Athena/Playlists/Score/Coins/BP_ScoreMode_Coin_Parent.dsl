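; Parent Blueprint for Score mode coins, which report collection to the score system when a player pawn or vehicle driver overlaps them.
; The server handles collection, sets CoinCollected, disables collision and destroys the coin after a delay.
; Clients loop a spin-and-pulse timeline while in the spin overlap, and OnRep_CoinCollected triggers pickup sound, shrink, material and particle effects.
; The construction script configures the optional light by platform and effects quality.
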(event EventBeginPlay
  (bind _spinoverlap (Variables|Default|GetSpinOverlap))
  ;[1 1 1] BEGIN PLAY
  (Networking|SwitchHasAuthority
    (:Authority
      ;[1 1 1] SERVER: Bind Collection Event
      (Collision|BindEventtoOnComponentBeginOverlap (Variables|Default|GetCollectOverlap) (EventDispatchers|CreateEvent :!event Custom|Collected)))
    (:Remote
      ;[1 1 1] CLIENT: Bind Spin Events
      (Collision|BindEventtoOnComponentBeginOverlap _spinoverlap (EventDispatchers|CreateEvent :!event Custom|BeginSpinning))
      (Collision|BindEventtoOnComponentEndOverlap _spinoverlap (EventDispatchers|CreateEvent :!event Custom|EndSpinning)))))

(event Custom|BeginSpinning
  (params (OverlappedComponent :type Primitive%20Component%20Object%20Reference) (OtherActor :type Actor%20Object%20Reference) (OtherComp :type Primitive%20Component%20Object%20Reference) (OtherBodyIndex :type Integer) (bFromSweep :type Boolean) (SweepResult :type Hit%20Result%20Structure%20%28by%20ref%29))
  ;[1 1 1] CLIENT ONLY SPINNING
  (Utilities|Casting|CastToFortPlayerPawnAthena OtherActor)
  (Variables|Default|SetCurrentRelativeScale (Variables|/Script/Engine.SceneComponent|GetRelativeScale3D (Variables|Default|GetSM_Coin)))
  (bind (_direction _risetime _pulsetime) (Timeline|SpinTL Play)
    (:Update
      (Transformation|SetRelativeRotation (Variables|Default|GetSM_Coin) (Math|Rotator|MakeRotator 0.0 0.0 (* _risetime (Variables|Default|GetRelativeRotationAmount))) false false)
      (Transformation|SetRelativeScale3D (Variables|Default|GetSM_Coin) (Math|Vector|Lerp(Vector) (Variables|Default|GetCurrentRelativeScale) (* (Variables|Default|GetCurrentRelativeScale) (Variables|Default|GetPulseSizeMultiplier)) _pulsetime)))
    (:Finished
      (bind _returnvalue (Utilities|Time|SetTimerbyEvent (Utilities|Select :Index _direction :Forward (delegate Custom|LoopForwards) :Backward (delegate Custom|LoopBackwards)) (Variables|Default|GetDelayBetweenSpin)))
      (Variables|Default|SetLoopTimerHandle _returnvalue))
    (:Start%20Spin
      (switch Utilities|FlowControl|Switch|SwitchonETimelineDirection _direction
        (:Forward
          (Audio|PlaySoundatLocation (Variables|Sound|GetCoinSpinSound) (Transformation|GetWorldLocation (Variables|Default|GetSM_Coin))))))))

(event Custom|LoopForwards
  ;[1 1 1] CLIENT ONLY SPINNING
  (Timeline|SpinTL PlayFromStart))

(event Custom|LoopBackwards
  ;[1 1 1] CLIENT ONLY SPINNING
  (Timeline|SpinTL ReverseFromEnd))

(event Custom|EndSpinning
  (params (OverlappedComponent :type Primitive%20Component%20Object%20Reference) (OtherActor :type Actor%20Object%20Reference) (OtherComp :type Primitive%20Component%20Object%20Reference) (OtherBodyIndex :type Integer))
  ;[1 1 1] CLIENT ONLY SPINNING
  (Components|Timeline|Stop (Variables|BP_ScoreMode_Coin_Parent|GetSpinTL))
  (Utilities|Time|ClearandInvalidateTimerbyHandle (Variables|Default|GetLoopTimerHandle)))

(event Custom|Collected
  (params (OverlappedComponent :type Primitive%20Component%20Object%20Reference) (OtherActor :type Actor%20Object%20Reference) (OtherComp :type Primitive%20Component%20Object%20Reference) (OtherBodyIndex :type Integer) (bFromSweep :type Boolean) (SweepResult :type Hit%20Result%20Structure%20%28by%20ref%29))
  (bind _playerwhocollectedcoin (Variables|Default|GetPlayerWhoCollectedCoin))
  ;[0.2392 0.5093 1] SERVER ONLY: COLLECTION
  (bind _asfort_athena_vehicle (Utilities|Casting|CastToFortAthenaVehicle OtherActor)
    (:then
      (Variables|Default|SetPlayerWhoCollectedCoin (Vehicle|GetDriver _asfort_athena_vehicle)))
    (:CastFailed
      (bind _asfort_player_pawn_athena (Utilities|Casting|CastToFortPlayerPawnAthena OtherActor))
      (Variables|Default|SetPlayerWhoCollectedCoin _asfort_player_pawn_athena))) ; Allow Vehicles to pickup coins
  (Utilities|IsValid _playerwhocollectedcoin
    (:Is%20Valid
      (Score|OnScoreActorCollected (Variables|/Script/Engine.Pawn|GetPlayerState _playerwhocollectedcoin))
      (Variables|Default|SetCoinCollected true)
      (Collision|SetActorEnableCollision false)
      (Utilities|FlowControl|Delay (+ (Components|Timeline|GetTimelineLength (Variables|BP_ScoreMode_Coin_Parent|GetPickedUp_TL)) 1.0))
      (Actor|DestroyActor))))

(event Custom|CoinCollectedFX
  (bind (_relativerotation_roll _relativerotation_pitch _relativerotation_yaw) (Math|Rotator|BreakRotator (Variables|/Script/Engine.SceneComponent|GetRelativeRotation_Roll (Variables|Default|GetSM_Coin))))
  ;[1 1 1] CLIENT ONLY: Collection FX
  (CallFunction|EndSpinning 0)
  (bind _returnvalue (Rendering|Material|CreateDynamicMaterialInstance (Variables|Default|GetSM_Coin) 0 0 "None" :!class /Script/Engine.PrimitiveComponent))
  (Variables|Default|SetMID0 _returnvalue)
  (bind _returnvalue_2 (Rendering|Material|CreateDynamicMaterialInstance (Variables|Default|GetSM_Coin) 1 0 "None" :!class /Script/Engine.PrimitiveComponent))
  (Variables|Default|SetMID1 _returnvalue_2)
  (Audio|PlaySoundatLocation (Variables|Sound|GetCoinCollectionSound) (Transformation|GetActorLocation))
  (Variables|Default|SetPreShrunkScale (Variables|/Script/Engine.SceneComponent|GetRelativeScale3D (Variables|Default|GetSM_Coin)))
  (Variables|Default|SetCoinRelativeTransform (Variables|/Script/Engine.SceneComponent|GetRelativeLocation (Variables|Default|GetSM_Coin)))
  (Variables|Default|SetCoinRelativeRotationZ _relativerotation_yaw)
  (bind _light (Variables|Default|GetLight)
    (:then
      (Development|SetHiddeninGame _light true false))
    (:else
      _))
  (bind (_direction _scale _zrotation _bobheight _location) (Timeline|PickedUp_TL PlayFromStart)
    (:Update
      (Transformation|SetRelativeLocation (Variables|Default|GetSM_Coin) (+ _location (Variables|Default|GetCoinRelativeTransform)) false false)
      (Transformation|SetRelativeScale3D (Variables|Default|GetSM_Coin) (Math|Vector|Lerp(Vector) (Variables|Default|GetPreShrunkScale) "0, 0, 0" _scale))
      (Transformation|AddRelativeRotation (Variables|Default|GetSM_Coin) (Math|Rotator|MakeRotator 0.0 0.0 (+ _zrotation (Variables|Default|GetCoinRelativeRotationZ))) false false)
      (Rendering|Material|SetVectorParameterValue (Variables|Default|GetMID0) "BobHeightMax" (Math|Color|Lerp(LinearColor) :B (Variables|Default|GetBobHeightMax) :Alpha _bobheight) :!class /Script/Engine.MaterialInstanceDynamic)
      (Rendering|Material|SetVectorParameterValue (Variables|Default|GetMID0) "BobHeightMin" (Math|Color|Lerp(LinearColor) :B (Variables|Default|GetBobHeightMin) :Alpha _bobheight) :!class /Script/Engine.MaterialInstanceDynamic)
      (Rendering|Material|SetScalarParameterValue (Variables|Default|GetMID0) "GlowScale" _bobheight :!class /Script/Engine.MaterialInstanceDynamic))
    (:Finished
      (Development|SetHiddeninGame (Variables|Default|GetSM_Coin) true false))
    (:SpawnParticles
      (Effects|SpawnEmitteratLocation :EmitterTemplate (Variables|Default|GetCoinCollectionEmitter) :Location (Transformation|GetSocketLocation (Variables|Default|GetSM_Coin) "FX") :Scale "1.000000,1.000000,1.000000" :bAutoDestroy true :PoolingMethod "AutoRelease" :bAutoActivateSystem true))))

(fn OnRep_CoinCollected
  (if (Variables|Default|GetCoinCollected)
    (Development|PrintString (Utilities|String|Append "AWARD POINTS TO: " (Utilities|GetDisplayName (Variables|Default|GetPlayerWhoCollectedCoin))))
    (Networking|SwitchHasAuthority
      (:Remote
        (CallFunction|CoinCollectedFX)))))

(fn ConstructionScript
  (Networking|CanExecuteCosmeticEvents
    (:True
      (if (or (Variables|Default|GetUseLight) (not (Fortnite|IsMobilePlatform)))
        (Utilities|EffectsQualityCheck 0
          (:Cinematic
            _)
          (:Epic
            _)
          (:High
            _)
          (:Medium
            (Components|DestroyComponent (Variables|Default|GetLight))
            (Transformation|SetRelativeScale3D (Variables|Default|GetSM_Coin) (Variables|Default|GetCoinScale)))
          (:Low
            _))
        (Rendering|Components|Light|SetLightColor (Variables|Default|GetLight) (Variables|Default|GetLightColor) true)
        (else
          _)))
    (:False
      _)))
