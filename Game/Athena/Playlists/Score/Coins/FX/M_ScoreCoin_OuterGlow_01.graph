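# Unlit alpha-composite material for the outer glow of a score coin, with streaking light rays and a gentle bob.
# Two panned light-ray samples multiply, partly blend toward a rounded copy and fade along UV1 V, forming outer streaks over an inner glow.
# That mask plus vertex red blends outer/inner colors and streaks/sphere mask, scaled for color and opacity with switchable depth fade.
# A time sine drives height offset and slight Z-axis rotation.
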Material /Game/Athena/Playlists/Score/Coins/FX/M_ScoreCoin_OuterGlow_01
graph {
 properties {
  bEnableResponsiveAA=true
  bTangentSpaceNormal=false
  blendMode=BLEND_AlphaComposite
  pixelDepthOffsetMode=PDOM_Legacy
  shadingModel=MSM_Unlit
  usedShadingModels=Unlit
 }
 nodes [
  n1 = Sine(None=Multiply(A=Time(), B=Constant() {r=0.4}))
  n2 = Constant3Vector() {constant=<rgba 0 0 1 0>}
  n3 = LinearInterpolate(A=Constant() {r=-0.01}, B=Constant() {r=0.01}, Alpha=@n1)
  Add(
   A=LinearInterpolate(
    A=VectorParameter() {parameterName=BobHeightMax, defaultValue=<rgba 0 0 10 0>}.RGB
    B=VectorParameter() {parameterName=BobHeightMin, defaultValue=<rgba 0 0 -2 0>}.RGB
    Alpha=@n1
   )
   B=RotateAboutAxis(NormalizedRotationAxis=@n2, RotationAngle=@n3, PivotPoint=ObjectPositionWS(), Position=WorldPosition().XYZ)
  )
  n4 = MaterialFunctionCall(
   Exponent=ScalarParameter() {defaultValue=2, parameterName="Exponent (S)"}
   Vertex / Pixel Normal=MaterialFunctionCall(Rotation Axis=@n2, Rotation Angle=@n3) {materialFunction=&FixRotateAboutAxisNormals}.New World Space Vertex Normal
  ) {materialFunction=&MF_Fresnel_FX}
  Multiply(B=@n4.Fresnel)
  n5 = VectorParameter() {parameterName=ColorOuter, defaultValue=<rgba 2 0.6328 0.1771 0>}
  Multiply(A=@n5.RGB, B=@n4.Fresnel)
  n6 = ComponentMask(None=TextureCoordinate() {coordinateIndex=1}) {g=true}
  n7 = ScalarParameter() {defaultValue=12, parameterName=InnerExponent}
  n8 = ScalarParameter() {defaultValue=5, parameterName=OuterExponent}
  n9 = Multiply(A=TextureCoordinate() {coordinateIndex=1}, B=AppendVector(A=ScalarParameter() {defaultValue=3, parameterName=Ua}, B=Constant() {r=1}))
  n10 = AppendVector(A=ScalarParameter() {defaultValue=0.2, parameterName=Speed Ua}, B=Constant())
  n11 = Multiply(A=TextureCoordinate() {coordinateIndex=1}, B=AppendVector(A=ScalarParameter() {defaultValue=4, parameterName=Ub}, B=Constant() {r=1}))
  n12 = AppendVector(A=ScalarParameter() {defaultValue=-0.294, parameterName=Speed Ub}, B=Constant())
  n13 = Multiply(
   A=TextureSample(UVs=Panner(Coordinate=@n9, Speed=@n10)) {texture=/Game/Effects/Fort_Effects/Textures/Outlander/T_InTheZone_LightRays_01}.R
   B=TextureSample(UVs=Panner(Coordinate=@n11, Speed=@n12)) {texture=&T_InTheZone_LightRays_01}.G
  )
  n14 = Round(None=@n13)
  n15 = ScalarParameter() {defaultValue=0.25, parameterName=lerp}
  n16 = Multiply(
   A=Multiply(A=Power(Base=@n6, Exp=@n8), B=LinearInterpolate(A=@n13, B=@n14, Alpha=@n15))
   B=ScalarParameter() {defaultValue=2, parameterName=StreakOuterScale}
  )
  n17 = Add(A=Multiply(A=Power(Base=@n6, Exp=@n7), B=ScalarParameter() {defaultValue=2, parameterName=InnerScale}), B=@n16)
  n18 = ScalarParameter() {defaultValue=3, parameterName=FinalExponent}
  n19 = Saturate(None=Add(A=VertexColor().R, B=Power(Base=@n17, Exp=@n18)))
  n20 = LinearInterpolate(A=@n5.RGB, B=VectorParameter() {parameterName=ColorInner, defaultValue=<rgba 10 4.5374 0.7292 0>}.RGB, Alpha=@n19)
  n21 = TextureCoordinate() {coordinateIndex=1}
  n22 = LinearInterpolate(
   A=@n16
   B=SphereMask(
    A=ComponentMask(None=@n21) {g=true}
    B=Constant() {r=0.5}
    Radius=ScalarParameter() {defaultValue=0.5, parameterName=Radius}
    Hardness=ScalarParameter() {defaultValue=0.3, parameterName=Hardness}
   )
   Alpha=@n19
  )
  n23 = ScalarParameter() {defaultValue=1, parameterName=GlowScale}
  Reroute(None=Multiply(A=Multiply(A=@n20, B=@n22), B=@n23))
  n24 = Multiply(A=@n23, B=@n22)
  StaticSwitchParameter(True=DepthFade(Opacity=@n24, FadeDistance=ScalarParameter() {defaultValue=40, parameterName=DepthFadeDistance}), False=@n24) {parameterName=EnableDepthFade}
 ]
}
