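# Unlit alpha-composite particle sprite material for coins in the Score playlist, drawing a soft sweep highlight.
# A centered radial mask is scaled by the dynamic parameter Sweep and inverted to set the center of a second sphere mask on offset UVs.
# Sweep thus shifts that mask along the UV diagonal, by an amount that varies with the first mask.
# The result, multiplied by particle color and alpha, feeds MF_FX_EmissiveAdjustments.
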Material /Game/Athena/Playlists/Score/Coins/Materials/M_CoinSweep_01
graph {
 properties {
  bUsedWithParticleSprites=true
  blendMode=BLEND_AlphaComposite
  pixelDepthOffsetMode=PDOM_Legacy
  shadingModel=MSM_Unlit
  usedShadingModels=Unlit
 }
 nodes [
  n1 = ParticleColor()
  n2 = SphereMask(
   A=TextureCoordinate()
   B=Constant() {r=0.5}
   Radius=ScalarParameter() {defaultValue=0.5, parameterName=Radius}
   Hardness=ScalarParameter() {parameterName=Hardness}
  )
  n3 = DynamicParameter() {paramNames [Sweep,Param2,Param3,Param4]}
  n4 = SphereMask(
   A=Add(A=TextureCoordinate()) {constB=0.5}
   B=OneMinus(None=Multiply(A=@n2, B=@n3.Sweep))
   Radius=ScalarParameter() {defaultValue=0.5, parameterName=SweepRadius}
   Hardness=ScalarParameter() {defaultValue=0.7, parameterName=SweepHardness}
  )
  MaterialFunctionCall(Emissive=Multiply(A=@n1.RGB, B=Multiply(A=@n1.A, B=@n4))) {materialFunction=&MF_FX_EmissiveAdjustments}
 ]
}
